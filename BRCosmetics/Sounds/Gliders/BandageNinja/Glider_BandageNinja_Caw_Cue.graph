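# Sound cue for the ambient caw vocal of the BandageNinja glider cosmetic.
# A random node picks one of three ambient vocal wave variants with equal weight.
# It plays quietly in the glider landing sound class, using the shared player glider open/close attenuation.
# A concurrency override caps simultaneous instances at a small count and blocks new ones beyond that.
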SoundCue /BRCosmetics/Sounds/Gliders/BandageNinja/Glider_BandageNinja_Caw_Cue
root=Random_0
nodes [
 {id=Random_0, children [WavePlayer_1,WavePlayer_2,WavePlayer_3], properties {weights [1,1,1]}}
 {id=WavePlayer_1, properties {soundWaveAssetPtr=&/Glider_BandageNinja_AmbientVocal_01}}
 {id=WavePlayer_2, properties {soundWaveAssetPtr=&/Glider_BandageNinja_AmbientVocal_02}}
 {id=WavePlayer_3, properties {soundWaveAssetPtr=&/Glider_BandageNinja_AmbientVocal_03}}
]
properties {
 attenuationSettings=/Game/Sounds/Attenuation/Fort_Player_Glider_OpenClose_Athena_Att
 bOverrideConcurrency=true
 maxDistance=10000
 soundClassObject=&Fort_Gliders_Land
 volumeMultiplier=0.2
 concurrencyOverrides {maxCount=2, resolutionRule=PreventNew}
}
