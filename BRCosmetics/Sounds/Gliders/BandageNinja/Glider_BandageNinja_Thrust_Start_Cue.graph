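# Sound cue for the BandageNinja glider's thrust start, playing a wing flap sound.
# A random node picks one of six equally weighted WingFlap waves, and a modulator slightly varies its volume on each play.
# Output is scaled down by a volume multiplier and uses the shared glider thrust attenuation and concurrency settings, with the glider land sound class.
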SoundCue /BRCosmetics/Sounds/Gliders/BandageNinja/Glider_BandageNinja_Thrust_Start_Cue
root=Modulator_0
nodes [
 {id=Modulator_0, children [Random_0], properties {volumeMax=1, volumeMin=0.85}}
 {id=Random_0, children [WavePlayer_0,WavePlayer_2,WavePlayer_3,WavePlayer_4,WavePlayer_8,WavePlayer_9], properties {weights [1,1,1,1,1,1]}}
 {id=WavePlayer_0, properties {soundWaveAssetPtr=&/Glider_BandageNinja_WingFlap_01}}
 {id=WavePlayer_2, properties {soundWaveAssetPtr=&/Glider_BandageNinja_WingFlap_02}}
 {id=WavePlayer_3, properties {soundWaveAssetPtr=&/Glider_BandageNinja_WingFlap_03}}
 {id=WavePlayer_4, properties {soundWaveAssetPtr=&/Glider_BandageNinja_WingFlap_04}}
 {id=WavePlayer_8, properties {soundWaveAssetPtr=&/Glider_BandageNinja_WingFlap_05}}
 {id=WavePlayer_9, properties {soundWaveAssetPtr=&/Glider_BandageNinja_WingFlap_06}}
]
properties {
 attenuationSettings=&Fort_Player_Glider_Thrust_Athena_Att
 maxDistance=12500
 soundClassObject=&Fort_Gliders_Land
 volumeMultiplier=0.3
 concurrencySet [&Athena_Glider_Thrust_Conc]
}
